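; Gameplay ability for the CherrySmoke Shot Block medallion, with debug-level logic and no damage blocking in this graph.
; On activation it starts listening through the EventProcessorSubsystem for events on the avatar Fort player pawn that match its Verb Filter.
; Each matching event prints a debug "Ouch" and applies the medallion's cooldown gameplay effect to the owner.
; A separate OnDamaged handler only prints the same string and is not bound here.
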(event Ability|EventActivateAbility
  (EventProcessorSubsystem|StartListeningToFilteredEvent 0 (CallFunction|GetAvatarasFortPlayerPawn) (Variables|Default|GetVerb%20Filter) false (delegate Custom|CustomEvent)))

(event Custom|CustomEvent
  (Development|PrintString "Ouch")
  (Ability|ApplyGameplayEffectToOwner "/CherrySmokeBoonsGameplay/Gameplay/ShotBlockMedallion/GE_CherrySmoke_Medallion_ShotBlock_Cooldown.GE_CherrySmoke_Medallion_ShotBlock_Cooldown_C"))

(event Custom|OnDamaged_Event
  (params (DamagedActor :type Actor%20Object%20Reference) (Damage :type Float%20%28single-precision%29) (InstigatedBy :type Controller%20Object%20Reference) (DamageCauser :type Actor%20Object%20Reference) (HitLocation :type Vector) (FHitComponent :type Primitive%20Component%20Object%20Reference) (BoneName :type Name) (Momentum :type Vector))
  (Development|PrintString "Ouch"))
